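# Physics asset for the Fruitcake backpack cosmetic, defining collision bodies and joint limits for its dangling bag and sheet bones.
# A capsule on the pack bone is the hub, with boxes on the bag and four upper sheet bones and spheres on the left and right sheets.
# Each dyn_ body is constrained directly to the pack rather than chained, with tight swing and twist limits that allow only small secondary motion.
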PhysicsAsset /BRCosmetics/Accessories/FORT_Backpacks/Backpack_F_MED_Fruitcake/Meshes/F_MED_Fruitcake_Pack_Physics
bodies {
 dyn_bag [{
   shapeType=Box
   extentX=5.6062
   extentY=5.331
   extentZ=2.5401
   shapeName=dyn_bag_box
   center=<2.7844 0.2767 0.7107>
   rotation=<rot 17.7076 1.1743 0>
  }]
 dyn_sheet_l [{shapeType=Sphere, radius=1.4114, shapeName=dyn_sheet_l_sphere, center=<0.1687 -0.0589 -1.7544>}]
 dyn_sheet_r [{shapeType=Sphere, radius=1.925, shapeName=dyn_sheet_r_sphere, center=<0.4831 -1.0724 -1.6371>}]
 dyn_sheet_up_01 [{
   shapeType=Box
   extentX=3.4598
   extentY=1.2722
   extentZ=8.6329
   shapeName=dyn_sheet_up_01_box
   center=<5.1902 -0.0578 -0.8543>
   rotation=<rot 7.2695 -90.3626 -71.5935>
  }]
 dyn_sheet_up_02 [{
   shapeType=Box
   extentX=3.6024
   extentY=1.3244
   extentZ=9.8424
   shapeName=dyn_sheet_up_02_box
   center=<4.4588 -0.0599 -2.3746>
   rotation=<rot 0.6695 -87.345 -55.5231>
  }]
 dyn_sheet_up_03 [{
   shapeType=Box
   extentX=3.0276
   extentY=1.7405
   extentZ=7.192
   shapeName=dyn_sheet_up_03_box
   center=<3.0997 0.6209 -2.3425>
   rotation=<rot 10.8688 -72.0724 -45.3977>
  }]
 dyn_sheet_up_04 [{
   shapeType=Box
   extentX=2.991
   extentY=1.197
   extentZ=6.2059
   shapeName=dyn_sheet_up_04_box
   center=<3.0084 0.6499 -1.2467>
   rotation=<rot 3.5504 -79.3655 -72.7547>
  }]
 pack [{
   shapeType=Capsule
   length=4.6935
   radius=7.25
   shapeName=pack_capsule
   center=<-1.2871 -10.958 -6.1613>
   rotation=<rot 88.4726 -0.0044 35.7307>
  }]
}
constraints [
 {bone1Name=dyn_bag, bone2Name=pack, swing1LimitDegrees=22, swing2LimitDegrees=25, twistLimitDegrees=2}
 {bone1Name=dyn_sheet_r, bone2Name=pack, swing1LimitDegrees=22, swing2LimitDegrees=22, twistLimitDegrees=10}
 {bone1Name=dyn_sheet_up_01, bone2Name=pack, swing1LimitDegrees=10, swing2LimitDegrees=21.52, twistLimitDegrees=5}
 {bone1Name=dyn_sheet_up_02, bone2Name=pack, swing1LimitDegrees=10, swing2LimitDegrees=15.76, twistLimitDegrees=5}
 {bone1Name=dyn_sheet_up_04, bone2Name=pack, swing1LimitDegrees=10, swing2LimitDegrees=18, twistLimitDegrees=5}
 {bone1Name=dyn_sheet_up_03, bone2Name=pack, swing1LimitDegrees=12, swing2LimitDegrees=18, twistLimitDegrees=5}
 {bone1Name=dyn_sheet_l, bone2Name=pack, swing1LimitDegrees=20, swing2LimitDegrees=15, twistLimitDegrees=5}
]
